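# Niagara effect for the Grilled Cheese backpack cosmetic, adding staged warm glows, flickering light and electricity at three skeletal-mesh sockets.
# Air_Glows_and_Light holds persistent socket sprites and meshes plus lights, driven by User.MidStage, User.FinalStage and a system integrator value.
# Electric_Coils bursts seeded Hermite-spline ribbons arcing between socket pairs, with endpoints rotating within per-socket spheres.
# Lightning_Sprites flashes short-lived sprites at the sockets; FinalStage sets spawn probability for both electric emitters.
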NiagaraSystem /BRCosmetics/Accessories/FORT_Backpacks/Backpack_M_MED_Grilled_Cheese/FX/NS_Backpack_GrilledCheese
dependencies {
 usedDataInterfaces [&NiagaraDataInterfaceCurve, &NiagaraDataInterfaceColorCurve, &NiagaraDataInterfaceArrayFloat]
 usedDynamicInputs [&SetBoolByFloatComparison]
 usedModules [&SystemState, &FN_Integrator]
}
emitters {
 Air_Glows_and_Light {
  emitterUpdateScript [
   &EmitterState()
   &SpawnBurst_Instantaneous(Spawn Count=3, Spawn Probability=HlslExpression() {hlslExpression="Emitter.LoopCount == 0"}, Use Spawn Probability=true)
  ]
  particleSpawnScript [
   &FN_SetSpawnGroup(Number of Spawn Groups=3)
   /Niagara/Modules/Spawn/Initialization/InitializeParticle(
    Color=<rgba 1.25 0.9431 0.5924 0.3>
    Lifetime=/Game/Effects/Niagara/DynamicInputs/Random/FN_SeededRandomRangeFloat(Minimum:NiagaraFloat=2, Maximum:NiagaraFloat=4, Random Seed:NiagaraInt32=$Particles.SpawnGroup)
    Mesh Scale=&VectorFromFloat(Value=&MaskFloatBySpawnGroup(
      Masked Float=0.15
      Passthrough Float=&MaskFloatBySpawnGroup(Masked Float=0.13, Passthrough Float=0.1, Spawn Group Mask=1)
     ))
    Module.Write Color=true
    Module.Write Lifetime=true
    Module.Write Mass=false
    Module.Write Material Random=true
    Module.Write Position=false
    Module.Write Scale=true
    Module.Write SpriteRotation=true
    Module.Write SpriteSize=true
    Sprite Rotation=&RandomRangeFloat(Minimum=0, Maximum=360)
    Sprite Size=&Vector2DFromFloat(Value=&MaskFloatBySpawnGroup(
      Masked Float=35
      Passthrough Float=&MaskFloatBySpawnGroup(Masked Float=30, Passthrough Float=28, Spawn Group Mask=1)
     ))
   )
   /SetVariables(Particles.MeshCameraOffset=-2) {isSetParameters=true}
  ]
  particleUpdateScript [
   &ParticleState(
    Kill Particles When Lifetime Has Elapsed=false
    Let Infinitely Lived Particles Die When Emitter Deactivates=true
    Loop Particles Lifetime=true
   )
   &ScaleColor(
    Scale Alpha=&FloatFromCurve(FloatCurve=DataInterface() {propertyValues {Curve {keys [<at 0.1 0.9 RCIM_Cubic>, <at 0.5 0.75 RCIM_Cubic>, {interpMode=RCIM_Cubic, leaveTangent=-10.3917, time=0.9, value=0.9}]}}})
    Scale RGB=&Multiply_VectorByFloat(
     Vector=&MakeVectorFromLinearColorRGB(Color=&ColorFromCurve(ColorCurve=DataInterface() {propertyValues {
         AlphaCurve {keys []}
         BlueCurve {keys [<at 0 1>,<at 0.198 0.6083>,<at 0.5 1.0391>,<at 0.605 0.9245>,<at 0.725 1.0234>,<at 1 1>]}
         GreenCurve {keys [<at 0 1>,<at 0.198 0.7872>,<at 0.5 1.4539>,<at 0.605 1.1632>,<at 0.725 1.4365>,<at 1 1>]}
         RedCurve {keys [<at 0 1>,<at 0.198 0.8>,<at 0.5 1.5>,<at 0.605 1.25>,<at 0.725 1.5>,<at 1 1>]}
        }}))
     Float=1
    )
    ScaleA=true
    ScaleRGB=true
   )
   &DynamicMaterialParameters(
    Index 0 Param 1=1
    Index 0 Param 2=&RandomRangeFloat(Minimum=0, Maximum=0.1)
    Index 0 Param 3=1
    Module.Param0WriteEnabled=true
    Module.Param1WriteEnabled=true
    Module.Param2WriteEnabled=true
    Module.Param3WriteEnabled=false
   )
   &SkeletalMeshLocation(
    Apply Sampled Orientation to Mesh Particles=true
    Bone / Socket Index=$Particles.SpawnGroup
    Inherit Velocity=false
    MaskSpawnGroup=false
    Mesh Sampling Type=::NewEnumerator5
    Skeletal Mesh=$User.SkeletalMesh
    Socket Sampling Mode=::NewEnumerator5
   )
   &FN_Light_Attributes(
    Color=$Particles.Color
    Create Color Attribute=true
    Create Exponent Attribute=true
    Create Position Attribute=true
    Create Radius Attribute=true
    Enable Particle Light Enable=true
    Exponent=1
    Light Enabled:NiagaraBool=/Niagara/DynamicInputs/Helpers/MaskBoolBySpawnGroup(Masked Bool:NiagaraBool=true, Passthrough Bool:NiagaraBool=false, Spawn Group Mask:NiagaraInt32=0)
    Position=&Add_Vector(
     A=Linked() {linkedVariable {name=Engine.Owner.Position, type=NiagaraPosition}}
     B=&RotateVectorByQuaternion(Vector To Rotate=<0 -10 0>, Quaternion=$Engine.Owner.Rotation)
    )
    Radius=&Multiply_Float(A=30, B=$User.MidStage)
    Scale Color=&Add_Float(
     A=&Multiply_Float(A=1.5, B=$System.FN_Integrator.Value)
     B=&FloatFromCurve(
      FloatCurve=DataInterface() {propertyValues {Curve {keys [<at 0 1>,<at 0.0351 1.3634>,<at 0.0806 0.5244>,<at 0.1124 1.3093>,<at 0.1564 0.6056>,<at 0.1942 1.5123>,<at 0.2337 0.971>,<at 0.2988 1.8507>,<at 0.3094 0.4702>,<at 0.3943 1.58>,<at 0.4337 0.5785>,<at 0.4701 1.6476>,<at 0.5208 0.1903>,<at 0.5883 1.1469>,<at 0.6368 0.565>,<at 0.7051 1.3364>,<at 0.7657 0.4973>,<at 0.8051 1.7424>,<at 0.9021 0.6868>,<at 1 1>]}}}
      Scale Curve=$User.FinalStage
     )
    )
   )
   /SetVariables(
    Particles.FlareSpriteColor:LinearColor=&MultiplyLinearColor(Linear Color=$Particles.Color, Scale Factor=<1.1 1.1 1.1 0.06>)
    Particles.FlareSpriteSize:Vector2f=&Vector2DFromFloat(Value=&Multiply_Float(
      A=&MaskFloatBySpawnGroup(
       Masked Float=65
       Passthrough Float=&MaskFloatBySpawnGroup(Masked Float=55, Passthrough Float=45, Spawn Group Mask=1)
      )
      B=&FloatFromCurve(
       FloatCurve=DataInterface() {propertyValues {Curve {keys [<at 0 1>,<at 0.1423 1.0958>,<at 0.2993 0.9039>,<at 0.4961 1.2434>,<at 0.7609 0.9039>,<at 1 1>]}}}
       Scale Curve=$System.FN_Integrator.Value
      )
     ))
   ) {isSetParameters=true}
   &SpriteRotationRate(Rotation Rate=8)
   &ScaleSpriteSize(Scale Factor=&Vector2DFromFloat(Value=$System.FN_Integrator.Value))
   /SetVariables(
    Particles.LightComponentColor:LinearColor=$Particles.Color
    Particles.LightComponentIntensity:NiagaraFloat=&Add_Float(
     A=&Multiply_Float(A=1.5, B=$User.MidStage)
     B=&FloatFromCurve(
      FloatCurve=DataInterface() {propertyValues {Curve {keys [<at 0 1>,<at 0.0493 3.1765>,<at 0.0797 0.3323>,<at 0.1266 3.1224>,<at 0.1555 0.4135>,<at 0.2085 3.3254>,<at 0.2479 2.784>,<at 0.3086 0.2782>,<at 0.3131 3.6637>,<at 0.4086 3.393>,<at 0.4329 0.3864>,<at 0.4844 3.4607>,<at 0.5118 0.3765>,<at 0.6026 2.96>,<at 0.636 0.3729>,<at 0.7193 3.1494>,<at 0.7648 0.3052>,<at 0.8193 3.5554>,<at 0.9013 0.4947>,<at 1 1>]}}}
      Scale Curve=$User.FinalStage
     )
    )
    Particles.LightComponentPosition:Vector3f=&Add_Vector(
     A=Linked() {linkedVariable {name=Engine.Owner.Position, type=NiagaraPosition}}
     B=&RotateVectorByQuaternion(Vector To Rotate=<0 -5 0>, Quaternion=$Engine.Owner.Rotation)
    )
   ) {isSetParameters=true}
  ]
  rendererClasses [
   &NiagaraLightRendererProperties
   &NiagaraSpriteRendererProperties
   &NiagaraMeshRendererProperties
   &NiagaraComponentRendererProperties
  ]
  renderers [
   {rendererClass=&NiagaraLightRendererProperties, rendererIndex=0}
   {rendererClass=&NiagaraSpriteRendererProperties, rendererIndex=1}
   {rendererClass=&NiagaraSpriteRendererProperties, rendererIndex=2}
   {rendererClass=&NiagaraMeshRendererProperties, rendererIndex=3}
   {rendererClass=&NiagaraComponentRendererProperties, rendererIndex=4}
  ]
 }
 Electric_Coils {
  emitterSpawnScript [
   /SetVariables(Emitter.RandomSeed:NiagaraInt32=/Niagara/DynamicInputs/Random/RandomInt(Randomness Mode:ENiagaraRandomnessMode=::NewEnumerator0, Random Seed:NiagaraInt32=0, Fixed Random Seed:NiagaraBool=false)) {isSetParameters=true}
   /SetVariables(Emitter.ParticlesPerGroup=7) {isSetParameters=true}
  ]
  emitterUpdateScript [
   /SetVariables(Emitter.NumberOfGroups=3) {isSetParameters=true}
   &EmitterState(
    Life Cycle Mode=::NewEnumerator1
    Loop Duration=&RandomRangeFloat(Minimum=0.15, Maximum=0.5)
    UseLoopDelay=false
   )
   &SpawnBurst_Instantaneous(
    Spawn Count=&Multiply_Int(A=$Emitter.ParticlesPerGroup, B=$Emitter.NumberOfGroups)
    Spawn Probability=$User.FinalStage
    Use Spawn Probability=true
   )
  ]
  particleSpawnScript [
   &FN_SetSpawnGroup(
    Inclusivity=true
    Normalize Spawn Group=true
    Particles Per Group:NiagaraInt32=$Emitter.ParticlesPerGroup
    Pattern=::NewEnumerator0
    Set Ribbon ID and Ribbon Link Order=true
   )
   /SetVariables(
    Particles.NeighborSocket=1
    Particles.SocketSelectionRandom:NiagaraInt32=&RandomRangeInt(
     Minimum=0
     Maximum=2
     Randomness Mode=::NewEnumerator1
     Random Seed=&Add_Integer(A=$Particles.SpawnGroup, B=$Emitter.LoopCount)
     Fixed Random Seed=true
    )
   ) {isSetParameters=true}
   &SphereLocation(
    Fixed Random Seed=true
    Hemisphere Y=true
    MaskSpawnGroup=false
    Override Seed=true
    Random Seed:NiagaraInt32=$Particles.SpawnGroup
    Randomness Mode:ENiagaraRandomnessMode=::NewEnumerator1
    Sphere Radius=&SelectFloatFromArray(Direct Array Index=$Particles.SocketSelectionRandom, Float Selection Array=DataInterface() {propertyValues {FloatData [7,6.25,4.75]}})
    Surface Only=false
    UseNonUniformScale=false
    UseOffset=false
    UseSphereOrientation=false
   )
   &SphereLocation(
    Fixed Random Seed=true
    Hemisphere Y=true
    MaskSpawnGroup=false
    Override Seed=true
    Random Seed:NiagaraInt32=$Particles.SpawnGroup
    Randomness Mode:ENiagaraRandomnessMode=::NewEnumerator1
    Sphere Radius=&SelectFloatFromArray(Direct Array Index=$Particles.NeighborSocket, Float Selection Array=DataInterface() {propertyValues {FloatData [7,6.25,4.75]}})
    Surface Only=false
    UseNonUniformScale=false
    UseOffset=false
    UseSphereOrientation=false
   ) {module=SphereLocation001}
   /SetVariables(
    Particles.EndSphereVector:Vector3f=$Output.SphereLocation001.SphereVector
    Particles.StartSphereVector:Vector3f=$Output.SphereLocation.SphereVector
   ) {isSetParameters=true}
   &InitializeRibbon(
    Color=<rgba 1.3 0.9247 0.387 1>
    Lifetime=&FN_SeededRandomRangeFloat(Minimum:NiagaraFloat=0.1, Maximum:NiagaraFloat=0.25, Random Seed:NiagaraInt32=$Particles.SpawnGroup)
    Module.Write Color=true
    Module.Write Lifetime=true
    Module.Write Mass=true
    Module.Write Material Random=true
    Module.Write Position=false
    Module.Write RibbonTwist=false
    Module.Write RibbonWidth=true
    Ribbon Width=&FN_SeededRandomRangeFloat(Minimum:NiagaraFloat=0.5, Maximum:NiagaraFloat=2.5, Random Seed:NiagaraInt32=$Particles.SpawnGroup)
   )
  ]
  particleUpdateScript [
   &ParticleState()
   &ScaleColor(
    Scale Alpha=&FloatFromCurve(FloatCurve=DataInterface() {propertyValues {Curve {keys [<at 0 0.25>,<at 0.15 1>,<at 1 0.35>]}}})
    Scale RGB=&VectorFromFloat(Value=&FloatFromCurve(FloatCurve=DataInterface() {propertyValues {Curve {keys [
          <at 0 1 RCIM_Cubic>
          {arriveTangent=-3.3333, interpMode=RCIM_Cubic, leaveTangent=-3.3333, time=0.05, value=2}
          {arriveTangent=-2.7322, interpMode=RCIM_Cubic, leaveTangent=-2.7322, time=0.15, value=0.5}
          {arriveTangent=2.5045, interpMode=RCIM_Cubic, leaveTangent=2.5045, time=0.416, value=1}
          {arriveTangent=-2.2322, interpMode=RCIM_Cubic, leaveTangent=-2.2322, time=0.5493, value=1.5}
          <at 0.64 0.5 RCIM_Cubic>
          {arriveTangent=1.3889, interpMode=RCIM_Cubic, leaveTangent=1.3889, time=0.8826, value=1.5}
          <at 1 1 RCIM_Cubic>
         ]}}}))
    ScaleA=true
    ScaleRGB=true
   )
   &SkeletalMeshLocation(
    Apply To Particle Position=false
    Bone / Socket Index=$Particles.SocketSelectionRandom
    Inherit Velocity=false
    Interpolate Mesh Sampling=false
    MaskSpawnGroup=false
    Mesh Sampling Type=::NewEnumerator5
    Skeletal Mesh=$User.SkeletalMesh
    Socket Sampling Mode=::NewEnumerator5
    Use Sampled Mesh Attribute as Ribbon ID=false
   )
   &SkeletalMeshLocation(
    Apply To Particle Position=false
    Bone / Socket Index=&ModuloInt(A=&Add_Integer(A=$Particles.NeighborSocket, B=$Particles.SocketSelectionRandom), B=3)
    Inherit Velocity=false
    Interpolate Mesh Sampling=false
    MaskSpawnGroup=false
    Mesh Sampling Type=::NewEnumerator5
    Skeletal Mesh=$User.SkeletalMesh
    Socket Sampling Mode=::NewEnumerator5
    Use Sampled Mesh Attribute as Ribbon ID=false
   ) {module=SkeletalMeshLocation001}
   &DynamicMaterialParameters(
    Index 0 Param 1=1
    Index 0 Param 2=1
    Index 0 Param 3=1
    Module.Param0WriteEnabled=true
    Module.Param1WriteEnabled=true
    Module.Param2WriteEnabled=false
    Module.Param3WriteEnabled=false
   )
   /SetVariables(Particles.Direction:Vector3f=&Multiply_VectorByFloat(
     Vector=&NormalizeVector(Vector To Normalize=/Niagara/DynamicInputs/Vectors/FindMidPointBetweenPositions(Vector1:Vector3f=$Output.SkeletalMeshLocation.SampledPosition, Vector2:Vector3f=$Output.SkeletalMeshLocation001.SampledPosition))
     Float=&RandomRangeFloat(Minimum=20, Maximum=50)
    )) {isSetParameters=true}
   /Game/Effects/Niagara/Modules/Curves/FN_HermiteCubicSpline_v1/FN_HermiteCubicSpline(
    Direction End:Vector3f=$Particles.Direction
    Direction Start:Vector3f=$Particles.Direction
    Position Blend:NiagaraFloat=$Particles.Normalized Spawn Group
    Position End:Vector3f=&Add_Vector(
     A=$Output.SkeletalMeshLocation001.SampledPosition
     B=&RotateVectorByQuaternion(
      Vector To Rotate=$Particles.Initial.EndSphereVector
      Quaternion=&MakeQuatFromAxisAngle(
       Axis=&RandomVector(Vector Scale=1, Randomness Mode=::NewEnumerator1, Random Seed=$Particles.SpawnGroup, Fixed Random Seed=true)
       AngleInDegrees=&Multiply_Float(A=-120, B=$Particles.NormalizedAge)
      )
     )
    )
    Position Start:Vector3f=&Add_Vector(
     A=$Output.SkeletalMeshLocation.SampledPosition
     B=&RotateVectorByQuaternion(
      Vector To Rotate=$Particles.Initial.StartSphereVector
      Quaternion=&MakeQuatFromAxisAngle(
       Axis=&RandomVector(Vector Scale=1, Randomness Mode=::NewEnumerator1, Random Seed=$Particles.SpawnGroup, Fixed Random Seed=true)
       AngleInDegrees=&Multiply_Float(A=90, B=$Particles.NormalizedAge)
      )
     )
    )
   )
   /SetVariables(Particles.Position=&Add_Vector(
     A=&RotateVectorByQuaternion(
      Vector To Rotate=&Subtract_Vector(A=$Output.FN_HermiteCubicSpline.Position, B=$Output.SkeletalMeshLocation001.SampledPosition)
      Quaternion=&MakeQuatFromAxisAngle(
       Axis=&Subtract_Vector(A=$Output.SkeletalMeshLocation.SampledPosition, B=$Output.SkeletalMeshLocation001.SampledPosition)
       AngleInDegrees=&RandomRangeFloat(Minimum=-0.5, Maximum=0.5)
      )
     )
     B=$Output.SkeletalMeshLocation001.SampledPosition
    )) {isSetParameters=true}
  ]
  rendererClasses [&NiagaraRibbonRendererProperties]
  renderers [{rendererClass=&NiagaraRibbonRendererProperties}]
 }
 Lightning_Sprites {
  emitterUpdateScript [
   &EmitterState()
   &SpawnRate(
    Spawn Probability=$User.FinalStage
    SpawnRate=&RandomRangeFloat(Minimum=20, Maximum=30, Evaluation Type=::NewEnumerator1)
    Use Spawn Probability=true
   )
  ]
  particleSpawnScript [
   &FN_SetSpawnGroup(Number of Spawn Groups=3)
   &InitializeParticle(
    Color=<rgba 0.7448 0.6468 0.353 1>
    Lifetime=&UniformRangedFloat(Minimum=0.15, Maximum=0.3)
    Mesh Scale=<3 3 3>
    Module.Write Color=true
    Module.Write Lifetime=true
    Module.Write Mass=false
    Module.Write Material Random=true
    Module.Write Position=false
    Module.Write Scale=true
    Module.Write SpriteRotation=true
    Module.Write SpriteSize=true
    Sprite Rotation=&NormalizedAngleToDegrees(NormalizedAngle=&UniformRangedFloat(Minimum=0, Maximum=1))
    Sprite Size=&Vector2DFromFloat(Value=&MaskFloatBySpawnGroup(
      Masked Float=&UniformRangedFloat(Minimum=18, Maximum=20)
      Passthrough Float=&MaskFloatBySpawnGroup(
       Masked Float=&UniformRangedFloat(Minimum=16, Maximum=18)
       Passthrough Float=&UniformRangedFloat(Minimum=12, Maximum=14)
       Spawn Group Mask=1
      )
     ))
   )
   /SetVariables(Particles.SubImageIndex=&UniformRangedFloat(Minimum=0, Maximum=3)) {isSetParameters=true}
   &CameraOffset(Camera Offset Amount=&RandomRangeFloat(Minimum=-4, Maximum=8))
  ]
  particleUpdateScript [
   &ParticleState()
   &DynamicMaterialParameters(
    Index 0 Param 1=&RandomRangeFloat(Minimum=0, Maximum=1)
    Index 0 Param 2=&RandomRangeFloat(Minimum=0, Maximum=0.1)
    Index 0 Param 3=&FloatFromCurve(FloatCurve=DataInterface() {propertyValues {Curve {keys [<at 0 1 RCIM_Cubic>,<at 0.2 0.025 RCIM_Cubic>,<at 1 1>]}}})
    Module.Param0WriteEnabled=true
    Module.Param1WriteEnabled=true
    Module.Param2WriteEnabled=true
    Module.Param3WriteEnabled=false
   )
   &SkeletalMeshLocation(
    Bone / Socket Index=$Particles.SpawnGroup
    Inherit Velocity=false
    MaskSpawnGroup=false
    Mesh Sampling Type=::NewEnumerator5
    Skeletal Mesh=$User.SkeletalMesh
    Socket Sampling Mode=::NewEnumerator5
   )
   &ScaleColor(
    Scale Alpha=&FloatFromCurve(FloatCurve=DataInterface() {propertyValues {Curve {keys [<at 0.75 1 RCIM_Cubic>, {arriveTangent=-10.3917, interpMode=RCIM_Cubic, leaveTangent=-10.3917, tangentMode=RCTM_User, time=1, value=0}]}}})
    Scale RGB=&Multiply_VectorByFloat(
     Vector=&MakeVectorFromLinearColorRGB(Color=&ColorFromCurve(ColorCurve=DataInterface() {propertyValues {
         AlphaCurve {keys []}
         BlueCurve {keys [<at 0 1>, {arriveTangent=-0.1823, interpMode=RCIM_Cubic, leaveTangent=-0.1823, time=0.15, value=1.7083}, <at 1 0.8177 RCIM_Cubic>]}
         GreenCurve {keys [<at 0 1>, {arriveTangent=-0.0091, interpMode=RCIM_Cubic, leaveTangent=-0.0091, time=0.15, value=2}, <at 1 0.9909 RCIM_Cubic>]}
         RedCurve {keys [<at 0 1>,<at 0.15 1.9903 RCIM_Cubic>,<at 1 1 RCIM_Cubic>]}
        }}))
     Float=1
    )
    ScaleA=true
    ScaleRGB=true
   )
  ]
  rendererClasses [&NiagaraSpriteRendererProperties]
  renderers [{rendererClass=&NiagaraSpriteRendererProperties}]
 }
}
summary {
 systemName=NS_Backpack_GrilledCheese
 emitters [
  {emitterName=Lightning_Sprites, rendererClasses [&NiagaraSpriteRendererProperties]}
  {
   emitterName=Air_Glows_and_Light
   rendererClasses [
    &NiagaraLightRendererProperties
    &NiagaraSpriteRendererProperties
    &NiagaraMeshRendererProperties
    &NiagaraComponentRendererProperties
   ]
  }
  {emitterName=Electric_Coils, rendererClasses [&NiagaraRibbonRendererProperties]}
 ]
 userVariables [
  User.FinalStage {type=NiagaraFloat, defaultValue=0, description=""}
  User.MidStage {type=NiagaraFloat, defaultValue=0, description=""}
  User.RibbonIDOffset {type=NiagaraInt32}
  User.SkeletalMesh {
   type=NiagaraDataInterfaceSkeletalMesh
   defaultValue {
    struct=NiagaraExt_VariableValue_DataInterface
    value {dataInterface=&self:NiagaraDataInterfaceSkeletalMesh_1}
   }
  }
 ]
}
